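# Conversation database for the Bulma NPC, defining the greeting and the options that follow it.
# The Bulma entry point leads to one Speech node with a default line and contextual overrides for a first-ever conversation or three specific player outfits, including Goku and Vegeta.
# First-meeting priority uses a curve row, while the outfit lines clear the curve.
# The speech then branches through link nodes into the shared NPCLibrary Quests and BasicServices conversations.
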ConversationDatabase /NPCLibrary/NPCs/Bulma/NPCLibrary_Conversation_Bulma
entry_points [{tag=Conversation.Entry.Bulma, destinations [#1]}]
nodes [
 ConversationEntryPointNode_7 {
  id=#1
  class=ConversationEntryPointNode
  outputs [#2]
  properties {entryTag {tagName=Conversation.Entry.Bulma}}
 }
 Speech_C_4 {
  id=#2
  class=Speech_C
  outputs [#3, #4]
  properties {
   nodeName=""
   generalConfig {
    defaultMessage="Time to relax!"
    contextualMessages [
     {
      message=It is a lot of trouble fixing machines you know.
      contextRequirements [{requirement {class=FortControllerRequirement_IsFirstEverConversationWithNPC}}]
      priority {curve {rowName=Tandem.Speech.Priority.FirstEverInteraction}}
     }
     {
      message="Goku, if you're hungry, why don't you try fishing?"
      contextRequirements [{requirement {
         class=FortControllerRequirement_HasCID
         properties {allowedCIds [/BRCosmetics/Athena/Items/Cosmetics/Characters/CID_A_462_Athena_Commando_M_Stamina]}
        }}]
      priority {value=0, curve {curveTable=null, rowName=None}}
     }
     {
      message="Oh Vegeta, I thought you were training."
      contextRequirements [{requirement {
         class=FortControllerRequirement_HasCID
         properties {allowedCIds [/BRCosmetics/Athena/Items/Cosmetics/Characters/CID_A_463_Athena_Commando_M_StaminaVigor]}
        }}]
      priority {value=0, curve {curveTable=null, rowName=None}}
     }
     {
      message=There seems to be a lot of unusual food on this island.
      contextRequirements [{requirement {
         class=FortControllerRequirement_HasCID
         properties {allowedCIds [/BRCosmetics/Athena/Items/Cosmetics/Characters/CID_A_464_Athena_Commando_M_StaminaCat]}
        }}]
      priority {value=0, curve {curveTable=null, rowName=None}}
     }
    ]
   }
  }
 }
 ConversationLinkNode_0 {id=#4, class=ConversationLinkNode, properties {remoteEntryTag {tagName=Conversation.Entry.BasicServices}}}
 ConversationLinkNode_1 {id=#3, class=ConversationLinkNode, properties {remoteEntryTag {tagName=Conversation.Entry.NPCLibrary.Quests}}}
]
